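# Niagara skydiving contrail cosmetic that shows a large glowing coin under the player and a ring of seven small coins above.
# Emitters follow a skeletal-mesh bone, with burst glows and curves on system age driving the intro.
# System variant index picks the enlarged, camera-facing main coin, rotates the ring and adjusts colors.
# Contrail setup's dive value and paraglide lean drive coin motion and flickering lightning trails, with separate front-end preview tuning.
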NiagaraSystem /BRCosmetics/Athena/Effects/Particles/Skydiving/NS_Contrail_SatireCane_Ode/NS_Contrail_SatireCane_Ode
dependencies {
 usedDataInterfaces [
  &NiagaraDataInterfaceCurve
  &NiagaraDataInterfaceMeshRendererInfo
  &NiagaraDataInterfaceParticleRead
  &NiagaraDataInterfaceCamera
  &NiagaraDataInterfaceVectorCurve
  &NiagaraDataInterfaceColorCurve
  &NiagaraDataInterfaceSkeletalMesh
  &NiagaraDataInterfaceArrayColor
 ]
 usedDynamicInputs [&MergeHelper_ColorSwitch]
 usedModules [
  &self:SystemSpawnScript.SystemScriptSource.SystemScriptGraph.NiagaraNodeAssignment_0.SetVariables_#1
  &self:SystemSpawnScript.SystemScriptSource.SystemScriptGraph.NiagaraNodeAssignment_1.SetVariables_#2
  &self:SystemSpawnScript.SystemScriptSource.SystemScriptGraph.NiagaraNodeAssignment_2.SetVariables_#3
  &self:SystemSpawnScript.SystemScriptSource.SystemScriptGraph.NiagaraNodeAssignment_3.SetVariables_#4
  &SystemState
  &self:ColorToInteger
  /Game/Effects/Niagara/Modules/Contrails/FN_Contrail_Setup
  &FN_Integrator
  &self:SystemSpawnScript.SystemScriptSource.SystemScriptGraph.NiagaraNodeAssignment_9.SetVariables_#5
  &self:SystemSpawnScript.SystemScriptSource.SystemScriptGraph.NiagaraNodeAssignment_6.SetVariables_#6
  &self:SystemSpawnScript.SystemScriptSource.SystemScriptGraph.NiagaraNodeAssignment_7.SetVariables_#7
  &self:SystemSpawnScript.SystemScriptSource.SystemScriptGraph.NiagaraNodeAssignment_8.SetVariables_#8
  &self:SystemSpawnScript.SystemScriptSource.SystemScriptGraph.NiagaraNodeAssignment_11.SetVariables_#9
  &self:SystemSpawnScript.SystemScriptSource.SystemScriptGraph.NiagaraNodeAssignment_10.SetVariables_#10
 ]
}
emitters {
 CoinUnderPlayer_Burst {
  emitterUpdateScript [
   &EmitterState(Life Cycle Mode=::NewEnumerator1, Loop Behavior=::NewEnumerator1, Loop Duration=2.3, UseLoopDelay=false)
   &SpawnBurst_Instantaneous(Use Spawn Probability=false, UseLoopCountLimit=false)
   &SpawnBurst_Instantaneous(Spawn Count=6, Spawn Group=1, Use Spawn Probability=false, UseLoopCountLimit=false) {module=SpawnBurst_Instantaneous001}
   /SetVariables(Emitter.CoinParams:Vector4f=<0 1 1 1>) {isSetParameters=true}
  ]
  particleSpawnScript [
   /SetVariables(Particles.SpawnGroup=$Emitter.SpawnGroup) {isSetParameters=true}
   /SetVariables(Particles.VisibilityTag=$Emitter.SpawnGroup) {isSetParameters=true}
   &InitializeParticle(
    Lifetime=$Emitter.CurrentLoopDuration
    Mesh Scale Mode=::NewEnumerator3
    Mesh Uniform Scale=&Multiply_Float(A=4.8, B=$System.BigCoin_Scale)
    Sprite Size Mode=::NewEnumerator3
    Uniform Sprite Size=66.7608
    UsePositionOffset=true
    Write Material Random=true
   )
   &InitialMeshOrientation(Mesh Axis To Orient=<0 0 1>, Mesh Orientation Mode=::NewEnumerator2, Orientation Vector=<0 1 0>, Use Rotation Vector=true)
  ]
  particleUpdateScript [
   &SolveForcesAndVelocity(Clamp Velocity=false, Limit Acceleration=false)
   &SkeletalMeshLocation(
    Bone / Socket Index=4
    Bone Sampling Mode=::NewEnumerator4
    MaskSpawnGroup=false
    Sampled Position Offset=<0 40 0>
    Skeletal Mesh=$System.Skeletal Mesh
   )
   &ParticleState()
   &ScaleMeshSize(Scale Factor=&VectorFromFloat(Value=&FloatFromCurve(
      FloatCurve=DataInterface() {propertyValues {Curve {
         defaultValue=1
         keys [{interpMode=RCIM_Cubic, tangentMode=RCTM_User, time=0.5, value=0}, {interpMode=RCIM_Cubic, tangentMode=RCTM_User, time=0.75, value=1}]
        }}}
      CurveIndex=$System.Age
     )))
   &ScaleColor(
    Scale Alpha=&FloatFromCurve(
     FloatCurve=DataInterface() {propertyValues {Curve {
        defaultValue=0
        keys [
         {arriveTangent=9.0725, interpMode=RCIM_Cubic, leaveTangent=9.0725, tangentMode=RCTM_User, time=0, value=0}
         {arriveTangent=-2.5249, interpMode=RCIM_Cubic, leaveTangent=-2.5249, tangentMode=RCTM_User, time=0.1408, value=0.9847}
         {arriveTangent=-0.0405, interpMode=RCIM_Cubic, leaveTangent=-0.0405, tangentMode=RCTM_User, time=1, value=0}
        ]
       }}}
     Scale Curve=0.25
    )
    ScaleA=true
    ScaleRGB=true
   )
   &DynamicMaterialParameters(
    Index 0 Param 1=&FloatFromCurve(FloatCurve=DataInterface() {propertyValues {Curve {
        defaultValue=1
        keys [
         {arriveTangent=4.1785, interpMode=RCIM_Cubic, leaveTangent=4.1785, tangentMode=RCTM_User, time=0, value=0}
         {arriveTangent=1.1179, interpMode=RCIM_Cubic, leaveTangent=1.1179, tangentMode=RCTM_User, time=0.0477, value=1.9762}
         {arriveTangent=0.0224, interpMode=RCIM_Cubic, leaveTangent=0.0224, tangentMode=RCTM_User, time=1, value=1}
        ]
       }}})
    Index 0 Param 2=1
    Index 0 Param 3=4
    Index 0 Param 4=4
    Module.Index0Float0Write=true
    Module.Index0Float1Write=true
    Module.Index0Float2Write=true
    Module.Index0Float3Write=true
   )
   /SetVariables(Particles.LightShaftParams:Vector4f=&MakeVector4(
     X=&FloatFromCurve(
      FloatCurve=DataInterface() {propertyValues {Curve {keys [
          <at 0 0 RCIM_Cubic>
          {arriveTangent=-1.0373, interpMode=RCIM_Cubic, leaveTangent=-1.0373, tangentMode=RCTM_User, time=0.1252, value=2}
          <at 0.9911 0.9086 RCIM_Cubic>
         ]}}}
      Scale Curve=45
     )
     Y=&Add_Float(
      A=&FloatFromCurve(FloatCurve=DataInterface() {propertyValues {Curve {defaultValue=1, keys [{tangentMode=RCTM_User, time=0, value=0}, {tangentMode=RCTM_User, time=1, value=1}]}}})
      B=5
     )
     Z=0
     W=0
    )) {isSetParameters=true}
   &FN_Mesh_Attributes(
    Color=&MultiplyLinearColor(
     Linear Color=$System.GlowColor
     Scale Factor=&MakeVector4FromVectorAndFloat(
      "Vector (RGB)"=<1 1 1>
      "Float (Alpha)"=&FloatFromCurve(
       FloatCurve=DataInterface() {propertyValues {Curve {keys [<at 0 1 RCIM_Cubic>, {arriveTangent=-1, interpMode=RCIM_Cubic, leaveTangent=-1, time=0.3079, value=0.2007}, <at 1 0 RCIM_Cubic>]}}}
       Scale Curve=2.5
      )
     )
    )
    Create Color Attribute=true
    Write to Module Namespace=true
   )
   &FN_Mesh_Attributes(
    Color=&Multiply_LinearColorByFloat(
     Linear Color=&ColorFromCurve(ColorCurve=DataInterface() {propertyValues {
        AlphaCurve {keys [<at 0 1 RCIM_Cubic>,<at 0.4766 0 RCIM_Cubic>]}
        BlueCurve {keys [<at 0 5 RCIM_Cubic>,<at 1 1 RCIM_Cubic>]}
        GreenCurve {keys [<at 0 3.5333 RCIM_Cubic>,<at 1 1 RCIM_Cubic>]}
        RedCurve {keys [<at 0 3.029 RCIM_Cubic>,<at 1 1 RCIM_Cubic>]}
       }})
     Float=0.2
    )
    Create Color Attribute=true
    Create Scale Attribute=true
    Scale=&Multiply_VectorByFloat(
     Vector=&MakeVector(X=9, Y=0.1, Z=9)
     Float=&FloatFromCurve(
      FloatCurve=DataInterface() {propertyValues {Curve {
         defaultValue=1
         keys [{interpMode=RCIM_Cubic, tangentMode=RCTM_User, time=0.5, value=0}, {interpMode=RCIM_Cubic, tangentMode=RCTM_User, time=0.75, value=1}]
        }}}
      CurveIndex=$System.Age
      Scale Curve=$System.BigCoin_Scale
     )
    )
    Write to Module Namespace=true
   ) {module=FN_Mesh_Attributes001}
   &FN_Light_Attributes(
    Color=&Multiply_LinearColorByFloat(Linear Color=$System.GlowColor, Float=4)
    Create Color Attribute=true
    Create Position Attribute=true
    Create Radius Attribute=true
    Position=<0 0 40>
    Radius=&Multiply_Float(
     A=60
     B=&FloatFromCurve(
      FloatCurve=DataInterface() {propertyValues {Curve {
         defaultValue=1
         keys [{interpMode=RCIM_Cubic, tangentMode=RCTM_User, time=0.5, value=0}, {interpMode=RCIM_Cubic, tangentMode=RCTM_User, time=0.75, value=1}]
        }}}
      CurveIndex=$System.Age
     )
    )
    Scale Color=&FloatFromCurve(
     FloatCurve=DataInterface() {propertyValues {Curve {
        defaultValue=1
        keys [
         {interpMode=RCIM_Cubic, tangentMode=RCTM_User, time=0.5, value=0}
         {interpMode=RCIM_Cubic, tangentMode=RCTM_User, time=0.75, value=1}
         <at 2 0 RCIM_Cubic>
        ]
       }}}
     CurveIndex=$System.Age
    )
   )
   &ScaleColor(
    Scale Mode=::NewEnumerator1
    Scale RGBA=&MakeVector4_FromLinearColor(LinearColor=$System.GlowColor)
    ScaleRGBA=true
   ) {module=ScaleColor001}
  ]
  rendererClasses [&NiagaraMeshRendererProperties, &NiagaraLightRendererProperties]
  renderers [
   {rendererClass=&NiagaraMeshRendererProperties, rendererIndex=0}
   {rendererClass=&NiagaraMeshRendererProperties, rendererIndex=1}
   {rendererClass=&NiagaraLightRendererProperties, rendererIndex=2}
   {rendererClass=&NiagaraMeshRendererProperties, rendererIndex=3}
  ]
 }
 Coin_Under_Player {
  emitterUpdateScript [
   &EmitterState(
    Inactive Response=::NewEnumerator1
    Life Cycle Mode=::NewEnumerator1
    Loop Behavior=::NewEnumerator1
    Loop Delay=2.2
    Loop Duration Mode=::NewEnumerator1
    UseLoopDelay=false
   )
   &SpawnBurst_Instantaneous(Use Spawn Probability=false, UseLoopCountLimit=false)
  ]
  particleSpawnScript [
   &InitializeParticle(
    Color=<rgba 1 1 1 0.4>
    Mesh Scale=&Multiply_VectorByFloat(Vector=<1 0.25 1>, Float=&Multiply_Float(A=1.9, B=$System.BigCoin_Scale))
    Mesh Scale Mode=::NewEnumerator4
    Sprite Size Mode=::NewEnumerator3
    Uniform Sprite Size=210
    UsePositionOffset=false
    Write Material Random=true
   )
   &CameraOffset(Camera Offset Amount=5)
   &InitialMeshOrientation(Mesh Orientation Mode=::NewEnumerator2, Orientation Vector=<0 0 1>, Rotation=<0.5 0 0>, Use Rotation Vector=true)
  ]
  particleUpdateScript [
   &ParticleState(Kill Particles When Lifetime Has Elapsed=false, Let Infinitely Lived Particles Die When Emitter Deactivates=true)
   &SkeletalMeshLocation(
    Bone / Socket Index=4
    Bone Sampling Mode=::NewEnumerator4
    MaskSpawnGroup=false
    Sampled Position Offset=<0 40 0>
    Skeletal Mesh=$System.Skeletal Mesh
   )
   &ScaleMeshSize(Scale Factor=&Multiply_VectorByFloat(
     Vector=<1 1 1>
     Float=&FloatFromCurve(
      FloatCurve=DataInterface() {propertyValues {Curve {
         defaultValue=1
         keys [{interpMode=RCIM_Cubic, tangentMode=RCTM_User, time=0.5, value=0}, {interpMode=RCIM_Cubic, tangentMode=RCTM_User, time=0.75, value=1}]
        }}}
      CurveIndex=$System.Age
     )
    ))
   &ScaleColor(
    Scale Mode=::NewEnumerator1
    Scale RGBA=&Multiply_Vector4(
     A=&MakeVector4_FromLinearColor(LinearColor=$System.CoinColors)
     B=&MakeVector4FromVectorAndFloat(
      "Vector (RGB)"=<1 1 1>
      "Float (Alpha)"=&FloatFromCurve(
       FloatCurve=DataInterface() {propertyValues {Curve {
          defaultValue=1
          keys [{interpMode=RCIM_Cubic, tangentMode=RCTM_User, time=0.25, value=0}, {interpMode=RCIM_Cubic, tangentMode=RCTM_User, time=0.6, value=1}]
         }}}
       CurveIndex=$System.Age
       Scale Curve=0.2
      )
     )
    )
    ScaleRGBA=true
   ) {module=ScaleColor001}
   /SetVariables(Particles.VisibilityTag=$System.VariantIndex) {isSetParameters=true}
  ]
  rendererClasses [&NiagaraMeshRendererProperties]
  renderers [
   {rendererClass=&NiagaraMeshRendererProperties, rendererIndex=0}
   {rendererClass=&NiagaraMeshRendererProperties, rendererIndex=1}
   {rendererClass=&NiagaraMeshRendererProperties, rendererIndex=2}
   {rendererClass=&NiagaraMeshRendererProperties, rendererIndex=3}
   {rendererClass=&NiagaraMeshRendererProperties, rendererIndex=4}
   {rendererClass=&NiagaraMeshRendererProperties, rendererIndex=5}
   {rendererClass=&NiagaraMeshRendererProperties, rendererIndex=6}
  ]
 }
 Coins_Above_Player {
  emitterUpdateScript [
   &EmitterState(Life Cycle Mode=::NewEnumerator1, Loop Behavior=::NewEnumerator1, Loop Duration Mode=::NewEnumerator1, UseLoopDelay=false)
   &SpawnBurst_Instantaneous(Spawn Count=7, Use Spawn Probability=false, UseLoopCountLimit=false)
  ]
  particleSpawnScript [
   &PartitionParticles(
    ID to Partition:NiagaraInt32=&ReturnExecIndex()
    Number Of Partitions=7
    Write Partition to Ribbon ID=::NewEnumerator0
   )
   &InitializeParticle(
    Color=<rgba 1 0.8524 0.6458 1>
    Lifetime Max=1
    Lifetime Mode=::NewEnumerator1
    Mesh Scale=&Multiply_VectorByFloat(Vector=<0.15 0.22 0.15>, Float=$System.SmallCoins_Scale)
    Mesh Scale Mode=::NewEnumerator4
    Sprite Size Mode=::NewEnumerator1
    UsePositionOffset=false
    Write Material Random=true
   )
   /SetVariables(Particles.MeshIndex=$Particles.PartitionParticles.Partition) {isSetParameters=true}
   /SetVariables() {isSetParameters=true}
  ]
  particleUpdateScript [
   &ParticleState(Kill Particles When Lifetime Has Elapsed=false, Let Infinitely Lived Particles Die When Emitter Deactivates=true)
   &SolveForcesAndVelocity(Clamp Velocity=false, Limit Acceleration=false)
   /SetVariables(Particles.isMainCoin:NiagaraBool=&SetBoolByIntComparison(
     A=&MergeHelper_IntSwitch(0=2, 1=5, 2=4, 3=0, 4=6, 5=3, 6=1, Index=$System.VariantIndex, Num=::NewEnumerator5)
     B=$Particles.PartitionParticles.Partition
     Comparison Type=::NewEnumerator4
    )) {isSetParameters=true}
   &InterpolateOverTime(Target Vector Value:NiagaraPosition=$Particles.Position, Variable Type=::NewEnumerator2) {module=InterpolateOverTime003}
   &ScaleMeshSize(Scale Factor=&VectorFromFloat(Value=&FloatFromCurve(
      FloatCurve=DataInterface() {propertyValues {Curve {keys [<at 0.1 0 RCIM_Cubic>, {arriveTangent=3.4587, interpMode=RCIM_Cubic, leaveTangent=3.4587, tangentMode=RCTM_User, time=0.5, value=1}]}}}
      CurveIndex=$Engine.System.Age
      Scale Curve=&MakeCustomFloatFromBool(Bool=$Particles.isMainCoin, True Float=2, False Float=1)
     )))
   &ShapeLocation(
    MaskSpawnGroup=false
    Offset=&FN_FrontEndVector3Modifier(Gameplay=<0 15 120>, Front End=<0 15 110>)
    Offset Mode=::NewEnumerator0
    Override Seed=false
    Ring / Disc Distribution Mode:ENiagaraSphereDistributionMode=::NewEnumerator1
    Ring Radius=50
    Rotation Angle:NiagaraFloat=&MergeHelper_FloatSwitch(0=-0.035, 1=0.53, 2=-0.33, 3=-0.75, 4=8.39, 5=-0.175, 6=-0.89, Index=$System.VariantIndex, Num=::NewEnumerator5)
    Rotation Axis=<0 0 1>
    Rotation Mode=::NewEnumerator1
    Shape Primitive=::NewEnumerator4
    U Position=&ReturnNormalizedExecIndex(Normalized Index Includes One=false)
    UseOwnerScale=true
   ) {module=ShapeLocation002}
   &SkeletalMeshLocation(
    Bone / Socket Index=4
    Bone Sampling Mode=::NewEnumerator4
    MaskSpawnGroup=false
    Orientation Sampling=::NewEnumerator1
    Position Sampling=::NewEnumerator2
    Sampled Position Offset Coordinate Space:ENiagaraSkelMeshTransforms=::NewEnumerator2
    Skeletal Mesh=$System.Skeletal Mesh
   )
   &UpdateMeshOrientation(
    Rotation Rate=&Multiply_Float(
     A=&RandomRangeFloat(
      Minimum=0.8
      Maximum=1
      Randomness Mode=::NewEnumerator1
      Random Seed=&Add_Integer(A=15, B=&ReturnParticleIDAttributes())
     )
     B=&Lerp_Float(A=1, B=2, Alpha=$System.FN_Contrail_Setup.Dive Value)
    )
    Rotation Vector=&MakeCustomVectorFromBool(
     Bool=$Particles.isMainCoin
     True Vector=<0 0 0>
     False Vector=&Multiply_VectorByFloat(
      Vector=&RandomRangeVector(
       Minimum=<0.3 1 0.5>
       Maximum=<0.7 0.3 1>
       Randomness Mode=::NewEnumerator1
       Random Seed=&Add_Integer(A=18, B=&ReturnParticleIDAttributes())
       Fixed Random Seed=true
      )
      Float=&MakeCustomFloatFromBool(
       Bool=&RandomBool(
        Randomness Mode=::NewEnumerator1
        Random Seed=&Add_Integer(A=15, B=&ReturnParticleIDAttributes())
       )
       True Float=1
       False Float=-1
      )
     )
    )
    Write to Orientation:ENiagara_AttributeSamplingApplyOutput=::NewEnumerator1
   )
   &UpdateMeshOrientation(
    Facing Direction=&ReturnCameraVector(Camera Vectors=::NewEnumerator1, Coordinate Space=::Simulation)
    Facing Interpolation=false
    Orientation Method=::NewEnumerator1
    Override Incoming Orientation=1
    Override Quat=true
    Override Quaternion:Quat4f=$Particles.MainCoinOrientation
    TransformFacing=false
    TransformUp=false
    Write to Orientation:ENiagara_AttributeSamplingApplyOutput=::NewEnumerator1
   ) {module=UpdateMeshOrientation001}
   /SetVariables(
    Particles.MainCoinOrientation:Quat4f=$Output.UpdateMeshOrientation001.MeshOrientation
    Particles.MeshOrientation=/Niagara/DynamicInputs/TypeConversions/MakeCustomQuaternionFromBool(
     Bool:NiagaraBool=$Particles.isMainCoin
     True Quat:Quat4f=$Output.UpdateMeshOrientation001.MeshOrientation
     False Quat:Quat4f=$Output.UpdateMeshOrientation.MeshOrientation
    )
   ) {isSetParameters=true}
   /SetVariables(Particles.ParaGlideAlphaRandomized:NiagaraFloat=&FloatFromCurve(
     FloatCurve=DataInterface() {propertyValues {Curve {keys [<at 0.2 0>,<at 0.8 1>]}}}
     CurveIndex=&Add_Float(
      A=&RemapRange(
       Input Value To Remap=&ClampFloat(Float=&Multiply_Float(A=$User.ParaGlideLeanAlpha, B=&RandomRangeFloat(Minimum=1, Maximum=1.2)))
       Input Min=0.16
       Input Max=1
       Output Min=0
       Output Max=1
      )
      B=&RandomRangeFloat(Minimum=-0.2, Maximum=0.2)
     )
    )) {isSetParameters=true}
   /SetVariables(
    Particles.InitialLightningAlpha:NiagaraFloat=&FloatFromCurve(
     FloatCurve=DataInterface() {propertyValues {Curve {keys [{arriveTangent=-2.8285, interpMode=RCIM_Cubic, leaveTangent=-2.8285, tangentMode=RCTM_User, time=0.6, value=1}, <at 1.2 0 RCIM_Cubic>]}}}
     CurveIndex=$Engine.System.Age
    )
    Particles.InitialLightningAlpha001:NiagaraFloat=&FloatFromCurve(FloatCurve=DataInterface() {propertyValues {Curve {keys [<at 1.2 1>,<at 1.5 0>]}}}, CurveIndex=$Engine.System.Age)
   ) {isSetParameters=true}
   /SetVariables(Particles.LightningAlpha:NiagaraFloat=&Lerp_Float(
     A=&MakeCustomFloatFromBool(
      Bool=&SetBoolByFloatComparison(
       A=&ClampFloat(Float=&Add_Float(A=$Particles.ParaGlideAlphaRandomized, B=$Particles.InitialLightningAlpha))
       B=0.9
      )
      True Float=1
      False Float=&MakeCustomFloatFromBool(
       Bool=&SetBoolByFloatComparison(
        A=&ClampFloat(Float=&Add_Float(A=$Particles.ParaGlideAlphaRandomized, B=$Particles.InitialLightningAlpha))
        B=0.1
       )
       True Float=&Waveform(
        Waveform Count=::NewEnumerator1
        "Waveform [1]"=::NewEnumerator4
        "[1] Frequency"=&RandomRangeFloat(Minimum=6, Maximum=12)
        "[1] Amplitude Min/Max"=<0 1>
        "Waveform [2] Blend"=::NewEnumerator2
        "Waveform [2]"=::NewEnumerator8
        "[2] Input Position"=0
        "[2] Phase"=$Emitter.Age
        "[2] Phase Scale"=1
        "[2] Compound Frequency"=2
        "[2] Frequency"=&RandomRangeFloat(Minimum=2, Maximum=4)
        "[2] Amplitude Min/Max"=<0 1>
        "[2] Scale"=1
        "[2] Bias"=0
        "[2] Exponent"=1
        "[2] Pulse Duration"=0.1
        "[2] Clip Triangle Min/Max"=<0 0>
        "[2] Random Seed"=0
       )
       False Float=0
      )
     )
     B=1
     Alpha=0
    )) {isSetParameters=true}
   &FN_Mesh_Attributes(
    Color=&MakeLinearColorFromVectorAndFloat(
     "Vector (RGB)"=&MakeVectorFromLinearColorRGB(Color=<rgba 0.5 0.5 0.5 1>)
     "Float (Alpha)"=&Multiply_Float(
      A=&MakeCustomFloatFromBool(
       Bool=$User.bIsFrontEndPreview
       True Float=&Lerp_Float(A=0.15, B=0.3, Alpha=$System.FN_Contrail_Setup.Dive Value)
       False Float=&Multiply_Float(A=&Lerp_Float(A=0.5, B=1, Alpha=$System.FN_Contrail_Setup.Dive Value), B=0.8)
      )
      B=0.2
     )
    )
    Create Color Attribute=true
    Create Orientation Attribute=true
    Create Scale Attribute=true
    Orientation:Quat4f=&MakeQuaternion(
     Quaternion From=1
     Axis=<1 0 0>
     Angle=0
     Roll=&FN_FrontEndFloatModifier(
      Gameplay=0
      Front End=&Lerp_Float(A=0, B=-27, Alpha=$System.FN_Contrail_Setup.InterpolateOverTime_Float.Moving Average)
     )
     From=?
     To=?
     Angle Type=::NewEnumerator0
     Normalize Vectors=false
    )
    Scale=&Multiply_VectorByFloat(
     Vector=&Multiply_Vector(
      A=&Multiply_VectorByFloat(Vector=<0.9 0.9 1>, Float=&RandomRangeFloat(Minimum=1, Maximum=1.2))
      B=&MakeCustomVectorFromBool(
       Bool=$User.bIsFrontEndPreview
       True Vector=&Multiply_Vector(
        A=&Lerp_Vector(
         A=<0.4 0.4 0.45>
         B=<0.85 0.85 0.8>
         Alpha=&ClampFloat(Float=&Add_Float(A=$System.FN_Contrail_Setup.Dive Value, B=$Particles.InitialLightningAlpha001))
        )
        B=<1.5 1.5 3>
       )
       False Vector=&Lerp_Vector(
        A=<0.4 0.4 0.45>
        B=<0.85 0.85 0.8>
        Alpha=&ClampFloat(Float=&Add_Float(A=$System.FN_Contrail_Setup.Dive Value, B=$Particles.InitialLightningAlpha001))
       )
      )
     )
     Float=&Multiply_Float(
      A=$System.SmallCoins_Scale
      B=&FloatFromCurve(FloatCurve=DataInterface() {propertyValues {Curve {keys [<at 0 0 RCIM_Cubic>,<at 0.4 1 RCIM_Cubic>]}}}, CurveIndex=$Engine.System.Age)
     )
    )
    Write to Module Namespace=true
   ) {module=FN_Mesh_Attributes001}
   /SetVariables(Particles.Position=&Multiply_VectorByFloat(
     Vector=&Lerp_Vector(
      A=&AddVectorToPosition(
       Position=$Particles.SkeletalMeshLocation.SampledPosition
       Vector=Linked() {linkedVariable {name=Particles.ShapeLocation002.ShapePosition, type=NiagaraPosition}}
      )
      B=&AddVectorToPosition(
       Position=&AddVectorToPosition(
        Position=$Particles.SkeletalMeshLocation.SampledPosition
        Vector=&MergeHelper_VectorSwitch(
         0=&Multiply_VectorByFloat(Vector=&MakeVector(X=$System.Waveform1, Y=$System.Waveform2, Z=$System.Waveform3), Float=0.6)
         1=&Multiply_VectorByFloat(Vector=&MakeVector(X=$System.Waveform4, Y=$System.Waveform5, Z=$System.Waveform6), Float=0.5)
         2=&Multiply_VectorByFloat(Vector=&MakeVector(X=$System.Waveform7, Y=$System.Waveform8, Z=$System.Waveform9), Float=0.9)
         3=&Multiply_VectorByFloat(Vector=&MakeVector(X=$System.Waveform1, Y=$System.Waveform2, Z=$System.Waveform3), Float=0.7)
         4=&Multiply_VectorByFloat(Vector=&MakeVector(X=$System.Waveform7, Y=$System.Waveform8, Z=$System.Waveform9), Float=0.6)
         5=&Multiply_VectorByFloat(Vector=&MakeVector(X=$System.Waveform4, Y=$System.Waveform5, Z=$System.Waveform6), Float=0.5)
         6=&Multiply_VectorByFloat(Vector=&MakeVector(X=$System.Waveform1, Y=$System.Waveform2, Z=$System.Waveform3), Float=0.8)
         Index=$Particles.PartitionParticles.Partition
         Num=::NewEnumerator5
        )
       )
       Vector=Linked() {linkedVariable {name=Particles.ShapeLocation002.ShapePosition, type=NiagaraPosition}}
      )
      Alpha=$System.FN_Contrail_Setup.Dive Value
     )
     Float=&FloatFromCurve(
      FloatCurve=DataInterface() {propertyValues {Curve {
         defaultValue=2
         keys [
          {interpMode=RCIM_Cubic, tangentMode=RCTM_User, time=0, value=2}
          {interpMode=RCIM_Cubic, tangentMode=RCTM_User, time=0.9, value=0.95}
          {arriveTangent=0.0833, interpMode=RCIM_Cubic, leaveTangent=0.0833, time=1.2, value=1}
          <at 1.5 1 RCIM_Cubic>
         ]
        }}}
      CurveIndex=&MakeCustomFloatFromBool(
       Bool=$Particles.isMainCoin
       True Float=&Add_Float(A=$System.Age, B=-0.5)
       False Float=&Multiply_Float(
        A=&RandomRangeFloat(Minimum=1, Maximum=1.1)
        B=&Add_Float(A=$System.Age, B=&RandomRangeFloat(Minimum=-0.5, Maximum=0))
       )
      )
     )
    )) {isSetParameters=true}
   /SetVariables(
    Particles.TrailDyn_01:Vector4f=&MakeVector4(
     X=&RandomRangeFloat(Minimum=0, Maximum=1)
     Y=&RandomRangeFloat(Minimum=0, Maximum=10)
     Z=&Multiply_Float(A=30, B=&Lerp_Float(A=0.5, B=1, Alpha=$System.FN_Contrail_Setup.Dive Value))
     W=0
    )
    Particles.TrailDyn_03:Vector4f=&MakeVector4(
     X=&Lerp_Float(A=0, B=1.5, Alpha=&Ceil_Float(Float=$System.FN_Contrail_Setup.Dive Value))
     Y=&Lerp_Float(A=0, B=0.5, Alpha=&Ceil_Float(Float=$System.FN_Contrail_Setup.Dive Value))
     Z=0
     W=0
    )
   ) {isSetParameters=true}
   /SetVariables(Particles.InterpolateDynamic:Vector4f=&MakeFloat4FromVectorAndFloat("Vector (RGB)"=$Particles.InterpolateOverTime003.InterpolateOverTime_V3_.Moving Average, "Float (Alpha)"=1)) {isSetParameters=true}
   /SetVariables(Particles.Dark glow:LinearColor=$System.DarkGlow_Color) {isSetParameters=true}
   /SetVariables(Particles.Inter ring size:Vector3f=&Multiply_Vector(
     A=&VectorFromFloat(Value=&Multiply_Float(A=0.022, B=$System.SmallCoins_Scale))
     B=$Transient.MeshScaleFactor
    )) {isSetParameters=true}
   /SetVariables(Particles.Dark glow mesh size:Vector3f=&Multiply_VectorByFloat(
     Vector=<1 1 1>
     Float=&MakeFloatFromVector(
      Vector=&Multiply_Vector(
       A=&VectorFromFloat(Value=&Multiply_Float(A=0.0225, B=$System.SmallCoins_Scale))
       B=$Transient.MeshScaleFactor
      )
      Channel=::NewEnumerator0
     )
    )) {isSetParameters=true}
   /SetVariables(Particles.Inter ring color:LinearColor=&Multiply_LinearColorByFloat(Linear Color=$System.DarkGlow_Color, Float=1.5)) {isSetParameters=true}
   &FN_Mesh_Attributes(
    Color=$System.DarkGlow_Color
    Create Color Attribute=true
    Create Scale Attribute=true
    Scale=&Multiply_Vector(
     A=&VectorFromFloat(Value=&Multiply_Float(A=0.067, B=$System.SmallCoins_Scale))
     B=$Transient.MeshScaleFactor
    )
    Write to Module Namespace=true
   )
   /SetVariables(Particles.DynPar_GlowSpline:Vector4f=&MakeVector4(
     X=&FloatFromCurve(FloatCurve=DataInterface() {propertyValues {Curve {
         defaultValue=1
         keys [
          {arriveTangent=4.1785, interpMode=RCIM_Cubic, leaveTangent=4.1785, tangentMode=RCTM_User, time=0, value=0}
          {arriveTangent=1.1179, interpMode=RCIM_Cubic, leaveTangent=1.1179, tangentMode=RCTM_User, time=0.0477, value=1.9762}
          {arriveTangent=0.0224, interpMode=RCIM_Cubic, leaveTangent=0.0224, tangentMode=RCTM_User, time=1, value=1}
         ]
        }}})
     Y=1
     Z=4
     W=4
    )) {isSetParameters=true}
   &FN_Mesh_Attributes(
    Color=&MultiplyLinearColor(
     Linear Color=$System.GlowColor
     Scale Factor=&MakeVector4FromVectorAndFloat(
      "Vector (RGB)"=&VectorFromFloat(Value=&MakeCustomFloatFromBool(
        Bool=&SetBoolByIntComparison(A=$System.VariantIndex, B=3, Comparison Type=::NewEnumerator4)
        True Float=0.05
        False Float=0.5
       ))
      "Float (Alpha)"=&MakeCustomFloatFromBool(
       Bool=&SetBoolByIntComparison(A=$System.VariantIndex, B=3, Comparison Type=::NewEnumerator4)
       True Float=0.2
       False Float=0.2
      )
     )
    )
    Create Color Attribute=true
    Create Scale Attribute=true
    Scale=&Multiply_Vector(
     A=&VectorFromFloat(Value=&Multiply_Float(A=0.37, B=$System.SmallCoins_Scale))
     B=$Transient.MeshScaleFactor
    )
    Write to Module Namespace=true
   ) {module=FN_Mesh_Attributes002}
   /SetVariables(Particles.CameraDot:NiagaraFloat=&ClampFloat(Float=&DotProduct(
      Vector A=&ReturnCameraVector(Camera Vectors=::NewEnumerator1, Coordinate Space=::Simulation)
      Vector B=<0 0 -1>
     ))) {isSetParameters=true}
   &UpdateMeshOrientation(
    Facing Interpolation=true
    Facing Interpolation Rate=3
    Facing Position:NiagaraPosition=&AddVectorToPosition(
     Position=$Particles.Position
     Vector=&TransformVector(
      Vector=&FN_FrontEndVector3Modifier(
       Gameplay=&MakeVector(
        X=&Lerp_Float(
         A=&MakeFloatFromVector(
          Vector=/Niagara/DynamicInputs/Math/Clamp_Vector(
           Vector:Vector3f=&Multiply_VectorByFloat(
            Vector=&Multiply_Vector(
             A=&ReturnCameraVector(Camera Vectors=::NewEnumerator4, Coordinate Space=::Local, Origin Position=$Particles.Position)
             B=<-1 0 0>
            )
            Float=1
           )
           Min:Vector3f=<-1 -1 -1>
           Max:Vector3f=<1 1 1>
          )
          Channel=::NewEnumerator0
         )
         B=&MakeFloatFromVector(
          Vector=&Multiply_VectorByFloat(
           Vector=&Clamp_Vector(
            Vector:Vector3f=&NormalizeVector(
             Vector To Normalize=&Multiply_VectorByFloat(
              Vector=&Multiply_Vector(
               A=&ReturnCameraVector(Camera Vectors=::NewEnumerator4, Coordinate Space=::Local, Origin Position=$Particles.Position)
               B=<-1 0 0>
              )
              Float=1
             )
             Safe Normalize=true
             Safe Fallback Threshold=0
             Safe Fallback Vector=<0 0 0.001>
            )
            Min:Vector3f=<-1 -1 -1>
            Max:Vector3f=<1 1 1>
           )
           Float=0.5
          )
          Channel=::NewEnumerator0
         )
         Alpha=&FloatFromCurve(FloatCurve=DataInterface() {propertyValues {Curve {keys [<at 0.5 0 RCIM_Cubic>,<at 0.75 1 RCIM_Cubic>]}}}, CurveIndex=$Particles.CameraDot)
        )
        Y=&FloatFromCurve(
         FloatCurve=DataInterface() {propertyValues {Curve {keys [<at 0.5 0 RCIM_Cubic>, {arriveTangent=-2, interpMode=RCIM_Cubic, leaveTangent=-2, time=0.7, value=0.5}, <at 1 -1 RCIM_Cubic>]}}}
         CurveIndex=$Particles.CameraDot
        )
        Z=&Lerp_Float(
         A=1
         B=0.6
         Alpha=&FloatFromCurve(FloatCurve=DataInterface() {propertyValues {Curve {keys [<at 0.5 0 RCIM_Cubic>,<at 1 1 RCIM_Cubic>]}}}, CurveIndex=$Particles.CameraDot)
        )
       )
       Front End=<0 -0.5 1>
      )
      Destination Space=::Simulation
     )
    )
    Orientation Method=::NewEnumerator2
    Override Incoming Orientation=1
    Override Quat=true
    Override Quaternion:Quat4f=$Particles.LightningOrientation
    TransformFacingOrigin=false
    TransformFacingPosition=false
    TransformUp=false
    Up Direction:Vector3f=&NormalizeVector(
     Vector To Normalize=&Subtract_Position(
      A=&ReturnCameraPosition(Coordinate Space:ENiagaraCoordinateSpace=::Simulation, Camera Query:NiagaraDataInterfaceCamera=DataInterface())
      B=$Particles.Position
     )
     Safe Normalize=true
     Safe Fallback Threshold=0
     Safe Fallback Vector=<0 0 0.001>
    )
    Write to Orientation:ENiagara_AttributeSamplingApplyOutput=::NewEnumerator1
   ) {module=UpdateMeshOrientation002}
   /SetVariables(
    Particles.DetailLightningTrailColor:LinearColor=&MultiplyLinearColor(
     Linear Color=&HueShiftLinearColor(
      Hue=&RandomRangeFloat(Minimum=-0.01, Maximum=0.01)
      Saturation=&RandomRangeFloat(Minimum=0.8, Maximum=1.2)
      Value=&RandomRangeFloat(Minimum=0.8, Maximum=1.1)
      Linear Color=&MultiplyLinearColor(
       Linear Color=$System.GlowColor
       Scale Factor=&MakeVector4FromVectorAndFloat(
        "Vector (RGB)"=<0.7 0.7 0.7>
        "Float (Alpha)"=&ClampFloat(Float=&Add_Float(A=$Particles.InitialLightningAlpha, B=&Power(a=$User.ParaGlideLeanAlpha, b=4)))
       )
      )
     )
     Scale Factor=&MakeVector4FromVectorAndFloat(
      "Vector (RGB)"=&MakeCustomVectorFromBool(
       Bool=&SetBoolByIntComparison(A=$System.VariantIndex, B=3, Comparison Type=::NewEnumerator4)
       True Vector=<0 0 0>
       False Vector=<1 1 1>
      )
      "Float (Alpha)"=&Multiply_Float(
       A=$Particles.LightningAlpha
       B=&MakeCustomFloatFromBool(
        Bool=&SetBoolByIntComparison(A=$System.VariantIndex, B=3, Comparison Type=::NewEnumerator4)
        True Float=5
        False Float=1
       )
      )
     )
    )
    Particles.LightningTrailCameraOffset=-5
    Particles.LightningTrailColor:LinearColor=&MultiplyLinearColor(
     Linear Color=&HueShiftLinearColor(
      Hue=&RandomRangeFloat(Minimum=-0.01, Maximum=0.01)
      Saturation=&RandomRangeFloat(Minimum=0.8, Maximum=1)
      Value=&RandomRangeFloat(Minimum=0.9, Maximum=1.1)
      Linear Color=$System.GlowColor
     )
     Scale Factor=&MakeVector4FromVectorAndFloat(
      "Vector (RGB)"=&MakeCustomVectorFromBool(
       Bool=&SetBoolByIntComparison(A=$System.VariantIndex, B=3, Comparison Type=::NewEnumerator4)
       True Vector=<0.5 0.5 0.5>
       False Vector=<1 1 1>
      )
      "Float (Alpha)"=&Multiply_Float(
       A=&ClampFloat(Float=&Add_Float(A=$Particles.LightningAlpha, B=$Particles.InitialLightningAlpha))
       B=&MakeCustomFloatFromBool(
        Bool=&SetBoolByIntComparison(A=$System.VariantIndex, B=3, Comparison Type=::NewEnumerator4)
        True Float=10
        False Float=1
       )
      )
     )
    )
    Particles.LightningTrailDyn:Vector4f=&MakeVector4(
     X=&Add_Float(
      A=&Multiply_FloatByInt(Float=0.125, Integer=$Particles.Initial.PartitionParticles.Partition)
      B=&Frac_Float(Float=&Multiply_Float(
        A=$Engine.System.Age
        B=&Multiply_Float(A=&RandomRangeFloat(Minimum=0.1, Maximum=0.2), B=-1)
       ))
     )
     Y=0
     Z=0
     W=0
    )
   ) {isSetParameters=true}
   &FN_Mesh_Attributes(
    Create Orientation Attribute=true
    Create Scale Attribute=true
    Orientation:Quat4f=&LerpQuaternion(
     Quaternion A:Quat4f=&MultiplyQuaternion(
      Quaternion A=&OrientMeshToVector(
       Look At Direction=/Niagara/DynamicInputs/Vectors/CrossProduct(
        Vector A:Vector3f=<0 0 1>
        Vector B:Vector3f=&ReturnCameraVector(Camera Vectors=::NewEnumerator3, Coordinate Space=::Simulation)
        Normalize Incoming Vectors:NiagaraBool=true
       )
       Reference Axis Vector=<1 0 0>
      )
      Quaternion B=&LerpQuaternion(
       Quaternion A:Quat4f=&MakeQuaternion(
        Quaternion From=1
        Axis=<1 0 0>
        Angle=0
        Pitch=&Multiply_Float(
         A=&MakeFloatFromVector(Vector=$System.FN_Contrail_Setup.Direction Vector, Channel=::NewEnumerator1)
         B=-120
        )
        Roll=&Multiply_Float(
         A=&MakeFloatFromVector(Vector=$System.FN_Contrail_Setup.Direction Vector, Channel=::NewEnumerator0)
         B=-120
        )
        From=?
        To=?
        Angle Type=::NewEnumerator0
        Normalize Vectors=false
       )
       Quaternion B:Quat4f=<0 0 0 1>
       Lerp Factor:NiagaraFloat=&FloatFromCurve(
        FloatCurve=DataInterface() {propertyValues {Curve {keys [<at 0.3 0 RCIM_Cubic>,<at 0.8 1 RCIM_Cubic>]}}}
        CurveIndex=&ClampFloat(Float=&DotProduct(
          Vector A=&ReturnCameraVector(Camera Vectors=::NewEnumerator1, Coordinate Space=::World)
          Vector B=&TransformVector(Vector=<0 0 -1>, Destination Space=::World)
         ))
       )
       Enable Spherical Linear Interpolation:NiagaraBool=false
      )
     )
     Quaternion B:Quat4f=&MakeQuaternion(
      Quaternion From=2
      Axis=<1 0 0>
      Angle=0
      Pitch=90
      X=&Add_Float(
       A=&Multiply_Float(
        A=&MakeFloatFromInt(Int=/Niagara/DynamicInputs/Subtract/Subtract_Integer(
          A:NiagaraInt32=&MergeHelper_IntSwitch(0=0, 1=3, 2=2, 3=5, 4=4, 5=1, 6=6, Index=$System.VariantIndex, Num=::NewEnumerator5)
          B:NiagaraInt32=$Particles.PartitionParticles.Partition
         ))
        B=&Divide_Float(A=360, B=7)
       )
       B=200
      )
      Y=90
      From=?
      To=?
      Angle Type=::NewEnumerator0
      Normalize Vectors=false
     )
     Lerp Factor:NiagaraFloat=&FloatFromCurve(
      FloatCurve=DataInterface() {propertyValues {Curve {keys [<at 0.8 0 RCIM_Cubic>,<at 1 1 RCIM_Cubic>]}}}
      CurveIndex=&ClampFloat(Float=&DotProduct(
        Vector A=&ReturnCameraVector(Camera Vectors=::NewEnumerator1, Coordinate Space=::World)
        Vector B=&TransformVector(Vector=<0 0 -1>, Destination Space=::World)
       ))
      Scale Curve=0.8
     )
     Enable Spherical Linear Interpolation:NiagaraBool=false
    )
    Scale=&Multiply_Vector(
     A=&Multiply_Vector(
      A=$Particles.FN_Mesh_Attributes001.Scale
      B=&Multiply_VectorByFloat(
       Vector=&Multiply_Vector(
        A=&RandomRangeVector(Minimum=<1 1 1>, Maximum=<1 1 1.2>)
        B=&MakeVector(
         X=1
         Y=1
         Z=&Waveform(
          Global Input Position=&RandomRangeFloat(Minimum=0, Maximum=10)
          "Waveform [1]"=::NewEnumerator9
          "[1] Frequency"=3
          "[1] Amplitude Min/Max"=<0.9 1.2>
         )
        )
       )
       Float=&Multiply_Float(
        A=&MakeCustomFloatFromBool(Bool=$Particles.isMainCoin, True Float=1.8, False Float=1)
        B=&Lerp_Float(
         A=0.5
         B=1
         Alpha=/Niagara/DynamicInputs/Camera/CameraFade(
          Camera Query:NiagaraDataInterfaceCamera=DataInterface()
          Fade Offset:NiagaraFloat=100
          Fade Distance:NiagaraFloat=200
          Use Owner Position:NiagaraBool=true
         )
        )
       )
      )
     )
     B=&VectorFromCurve(
      VectorCurve=DataInterface() {propertyValues {
        XCurve {keys [<at 0 1>,<at 1 1>]}
        YCurve {keys [<at 0 1>,<at 1 1>]}
        ZCurve {keys [{arriveTangent=0.026, interpMode=RCIM_Cubic, leaveTangent=0.026, tangentMode=RCTM_User, time=0.1, value=0.5}, <at 0.4 1 RCIM_Cubic>]}
       }}
      CurveIndex=&Add_Float(A=$User.ParaGlideLeanAlpha, B=$Particles.InitialLightningAlpha001)
      Scale Curve=&VectorFromFloat(Value=&FN_FrontEndFloatModifier(Gameplay=1, Front End=0.5))
     )
    )
    Write to Module Namespace=true
   ) {module=FN_Mesh_Attributes003}
   /SetVariables(Particles.LightningGlowColor:LinearColor=&MultiplyLinearColor(
     Linear Color=$System.LightningGlowColor
     Scale Factor=&MakeVector4FromVectorAndFloat(
      "Vector (RGB)"=<1 1 1>
      "Float (Alpha)"=&Multiply_Float(
       A=&Multiply_Float(
        A=&ClampFloat(Float=&Add_Float(A=$Particles.LightningAlpha, B=$Particles.InitialLightningAlpha))
        B=&MakeCustomFloatFromBool(
         Bool=&SetBoolByIntComparison(A=$System.VariantIndex, B=3, Comparison Type=::NewEnumerator4)
         True Float=50
         False Float=1
        )
       )
       B=0.3
      )
     )
    )) {isSetParameters=true}
   /SetVariables(Particles.LightningOrientation:Quat4f=$Output.UpdateMeshOrientation002.MeshOrientation) {isSetParameters=true}
  ]
  rendererClasses [&NiagaraMeshRendererProperties]
  renderers [
   {rendererClass=&NiagaraMeshRendererProperties, rendererIndex=0}
   {rendererClass=&NiagaraMeshRendererProperties, rendererIndex=1}
   {rendererClass=&NiagaraMeshRendererProperties, rendererIndex=2}
   {rendererClass=&NiagaraMeshRendererProperties, rendererIndex=3}
   {rendererClass=&NiagaraMeshRendererProperties, rendererIndex=4}
   {rendererClass=&NiagaraMeshRendererProperties, rendererIndex=5}
   {rendererClass=&NiagaraMeshRendererProperties, rendererIndex=6}
   {rendererClass=&NiagaraMeshRendererProperties, rendererIndex=7}
   {rendererClass=&NiagaraMeshRendererProperties, rendererIndex=8}
  ]
 }
 Glow_Under {
  emitterUpdateScript [
   &EmitterState(
    Inactive Response=::NewEnumerator1
    Life Cycle Mode=::NewEnumerator1
    Loop Behavior=::NewEnumerator1
    Loop Delay=2.2
    Loop Duration Mode=::NewEnumerator1
    UseLoopDelay=false
   )
   &SpawnBurst_Instantaneous(Use Spawn Probability=false, UseLoopCountLimit=false)
  ]
  particleSpawnScript [
   &InitializeParticle(
    Mesh Scale=&Multiply_VectorByFloat(Vector=<1 1 0.5>, Float=&Multiply_Float(A=0.8, B=$System.BigCoin_Scale))
    Mesh Scale Mode=::NewEnumerator4
    Sprite Size Mode=::NewEnumerator3
    Uniform Sprite Size=210
    UsePositionOffset=false
    Write Material Random=true
   )
   &CameraOffset(Camera Offset Amount=5)
   &InitialMeshOrientation(Mesh Orientation Mode=::NewEnumerator2, Rotation=<0.25 0 0>, Use Rotation Vector=true)
  ]
  particleUpdateScript [
   &ParticleState(Kill Particles When Lifetime Has Elapsed=false, Let Infinitely Lived Particles Die When Emitter Deactivates=true)
   &SkeletalMeshLocation(
    Bone / Socket Index=4
    Bone Sampling Mode=::NewEnumerator4
    MaskSpawnGroup=false
    Sampled Position Offset=<0 40 0>
    Skeletal Mesh=$System.Skeletal Mesh
   )
   /SetVariables(Particles.Inter ring color:LinearColor=&Multiply_LinearColorByFloat(Linear Color=$System.DarkGlow_Color, Float=1.5)) {isSetParameters=true}
   /SetVariables(Particles.Dark glow:LinearColor=$System.DarkGlow_Color) {isSetParameters=true}
   &ScaleMeshSize(Scale Factor=&Multiply_VectorByFloat(
     Vector=<1 1 1>
     Float=&FloatFromCurve(
      FloatCurve=DataInterface() {propertyValues {Curve {
         defaultValue=1
         keys [{interpMode=RCIM_Cubic, tangentMode=RCTM_User, time=0.5, value=0}, {interpMode=RCIM_Cubic, tangentMode=RCTM_User, time=0.75, value=1}]
        }}}
      CurveIndex=$System.Age
     )
    ))
   &ScaleColor(
    Scale Mode=::NewEnumerator1
    Scale RGBA=&MakeVector4_FromLinearColor(LinearColor=$System.DarkGlow_Color)
    ScaleRGBA=true
   )
   /SetVariables(Particles.Inter ring size:Vector3f=&Multiply_VectorByFloat(
     Vector=&VectorFromFloat(Value=&FloatFromCurve(
       FloatCurve=DataInterface() {propertyValues {Curve {
          defaultValue=1
          keys [{interpMode=RCIM_Cubic, tangentMode=RCTM_User, time=0.5, value=0}, {interpMode=RCIM_Cubic, tangentMode=RCTM_User, time=0.75, value=1}]
         }}}
       CurveIndex=$System.Age
      ))
     Float=&Multiply_Float(A=0.25, B=$System.BigCoin_Scale)
    )) {isSetParameters=true}
   /SetVariables(Particles.Dark glow mesh size:Vector3f=&Multiply_VectorByFloat(
     Vector=&VectorFromFloat(Value=&FloatFromCurve(
       FloatCurve=DataInterface() {propertyValues {Curve {
          defaultValue=1
          keys [{interpMode=RCIM_Cubic, tangentMode=RCTM_User, time=0.5, value=0}, {interpMode=RCIM_Cubic, tangentMode=RCTM_User, time=0.75, value=1}]
         }}}
       CurveIndex=$System.Age
      ))
     Float=&Multiply_Float(A=0.25, B=$System.BigCoin_Scale)
    )) {isSetParameters=true}
   &FN_Mesh_Attributes(
    Color=&Multiply_LinearColorByFloat(Linear Color=$System.GlowColor, Float=0.5)
    Create Color Attribute=true
    Create Scale Attribute=true
    Scale=&Multiply_VectorByFloat(
     Vector=&VectorFromFloat(Value=&FloatFromCurve(
       FloatCurve=DataInterface() {propertyValues {Curve {
          defaultValue=1
          keys [{interpMode=RCIM_Cubic, tangentMode=RCTM_User, time=0.5, value=0}, {interpMode=RCIM_Cubic, tangentMode=RCTM_User, time=0.75, value=1}]
         }}}
       CurveIndex=$System.Age
      ))
     Float=&Multiply_Float(A=4.7, B=$System.BigCoin_Scale)
    )
    Write to Module Namespace=true
   ) {module=FN_Mesh_Attributes002}
   /SetVariables(Particles.DynPar_GlowSpline:Vector4f=&MakeVector4(
     X=&FloatFromCurve(FloatCurve=DataInterface() {propertyValues {Curve {
         defaultValue=1
         keys [
          {arriveTangent=4.1785, interpMode=RCIM_Cubic, leaveTangent=4.1785, tangentMode=RCTM_User, time=0, value=0}
          {arriveTangent=1.1179, interpMode=RCIM_Cubic, leaveTangent=1.1179, tangentMode=RCTM_User, time=0.0477, value=1.9762}
          {arriveTangent=0.0224, interpMode=RCIM_Cubic, leaveTangent=0.0224, tangentMode=RCTM_User, time=1, value=1}
         ]
        }}})
     Y=1
     Z=4
     W=4
    )) {isSetParameters=true}
  ]
  rendererClasses [&NiagaraMeshRendererProperties]
  renderers [
   {rendererClass=&NiagaraMeshRendererProperties, rendererIndex=0}
   {rendererClass=&NiagaraMeshRendererProperties, rendererIndex=1}
   {rendererClass=&NiagaraMeshRendererProperties, rendererIndex=2}
   {rendererClass=&NiagaraMeshRendererProperties, rendererIndex=3}
  ]
 }
 Lighting_Coin_Under_Player {
  emitterUpdateScript [
   &EmitterState(Life Cycle Mode=::NewEnumerator1, Loop Behavior=::NewEnumerator1, Loop Delay=0.42, Loop Duration Mode=::NewEnumerator1, UseLoopDelay=true)
   &SpawnRate(
    SpawnRate=&FloatFromCurve(
     FloatCurve=DataInterface() {propertyValues {Curve {keys [{arriveTangent=-3421.9607, interpMode=RCIM_Cubic, leaveTangent=-3421.9607, tangentMode=RCTM_User, time=0.75, value=300}, <at 1 50>]}}}
     CurveIndex=$System.Age
     Scale Curve=0.2
    )
    Use Spawn Probability=false
   )
  ]
  particleSpawnScript [
   &InitializeParticle(
    Color=<rgba 2 2 2 1>
    Lifetime=&RandomRangeFloat(Minimum=0.5, Maximum=0.7)
    Mesh Scale=&Multiply_VectorByFloat(
     Vector=&RandomRangeVector(Minimum=<0.5 0.5 0.75>, Maximum=<0.5 0.5 1>)
     Float=&Multiply_Float(
      A=&RandomRangeFloat(Minimum=0.5, Maximum=0.7)
      B=&FloatFromCurve(
       FloatCurve=DataInterface() {propertyValues {Curve {keys [<at 0.5 0.9>,<at 0.75 0.9>,<at 0.8 1>,<at 1 0.75>]}}}
       CurveIndex=$System.Age
       Scale Curve=$System.BigCoin_Scale
      )
     )
    )
    Mesh Scale Mode=::NewEnumerator4
    UsePositionOffset=false
    Write Material Random=true
   )
   &InitialMeshOrientation(
    Mesh Orientation Mode=::NewEnumerator7
    Rotation=&RandomRangeVector(Minimum=<-1 -1 -1>, Maximum=<1 1 1>)
    Use Rotation Vector=true
   )
   &CameraOffset(Camera Offset Amount=10)
  ]
  particleUpdateScript [
   &ParticleState()
   &SkeletalMeshLocation(
    Bone / Socket Index=4
    Bone Sampling Mode=::NewEnumerator4
    MaskSpawnGroup=false
    Position Sampling=::NewEnumerator2
    Sampled Position Offset=<0 40 0>
    Skeletal Mesh=$System.Skeletal Mesh
   ) {module=SkeletalMeshLocation001}
   &ShapeLocation(
    Additional Quaternion Rotation=false
    Apply To Particle Position=::NewEnumerator1
    Handle Radius=0
    Large Radius=&FloatFromCurve(
     FloatCurve=DataInterface() {propertyValues {Curve {
        defaultValue=1
        keys [
         {arriveTangent=0.3634, interpMode=RCIM_Cubic, leaveTangent=0.3634, tangentMode=RCTM_User, time=0.5, value=0}
         {arriveTangent=-0.2128, interpMode=RCIM_Cubic, leaveTangent=-0.2128, tangentMode=RCTM_User, time=0.85, value=1.1}
         <at 2 1 RCIM_Cubic>
        ]
       }}}
     CurveIndex=$System.Age
     Scale Curve=&Multiply_Float(A=50, B=$System.BigCoin_Scale)
    )
    MaskSpawnGroup=false
    Override Seed=false
    Rotation Mode=::NewEnumerator3
    Rotation Quaternion=&MultiplyQuaternion(Quaternion A=$Particles.SkeletalMeshLocation001.SampledRotation, Quaternion B=<1 0 0 1>)
    Scale Mode:ENiagara_ScaleMode=::NewEnumerator1
    Shape Primitive=::NewEnumerator3
   )
   &ScaleMeshSize(Scale Factor=&VectorFromFloat(Value=&FloatFromCurve(
      FloatCurve=DataInterface() {propertyValues {Curve {keys [
          {arriveTangent=1.9115, interpMode=RCIM_Cubic, leaveTangent=1.9115, tangentMode=RCTM_User, time=0, value=0}
          {arriveTangent=0.0224, interpMode=RCIM_Cubic, leaveTangent=0.0224, tangentMode=RCTM_User, time=1, value=1.5}
         ]}}}
      Scale Curve=&MakeCustomFloatFromBool(Bool=$User.bIsFrontEndPreview, True Float=1, False Float=1)
     )))
   &DynamicMaterialParameters(
    Index 0 Param 2=&RandomRangeFloat(Minimum=0, Maximum=0)
    Index 0 Param 4=13
    Index 1 Param 1=&FloatFromCurve(FloatCurve=DataInterface() {propertyValues {Curve {keys [<at 0 0.5>,<at 1 1>]}}})
    Index 1 Param 2=2
    Index 1 Param 3=&FloatFromCurve(FloatCurve=DataInterface() {propertyValues {Curve {keys [{interpMode=RCIM_Cubic, tangentMode=RCTM_Break, time=0.5, value=-0.5}, <at 1 0.1>]}}})
    Index 2 Param 1=&Multiply_Float(
     A=&FloatFromCurve(
      FloatCurve=DataInterface() {propertyValues {Curve {keys [{interpMode=RCIM_Cubic, leaveTangent=8.1085, tangentMode=RCTM_Break, time=0.8568, value=1}, {tangentMode=RCTM_User, time=1, value=0}]}}}
      Scale Curve=0.4
     )
     B=&Divide_Float(
      A=&MakeFloatFromVector(Vector=$Particles.Scale, Channel=::NewEnumerator0)
      B=&RandomRangeFloat(Minimum=20, Maximum=25)
     )
    )
    Index 2 Param 2=&FloatFromCurve(
     FloatCurve=DataInterface() {propertyValues {Curve {keys [
         {interpMode=RCIM_Cubic, tangentMode=RCTM_Break, time=0.7, value=0}
         {arriveTangent=-0.0859, interpMode=RCIM_Cubic, tangentMode=RCTM_Break, time=1, value=-1}
        ]}}}
     Scale Curve=50
    )
    Index 2 Param 3=&Waveform(
     Global Input Position=&RandomRangeFloat(Minimum=0, Maximum=15)
     Global Phase=$Particles.Age
     "Waveform [1]"=::NewEnumerator12
     "[1] Phase Scale"=2
     "[1] Frequency"=1
     "[1] Amplitude Min/Max"=<0 1>
    )
    Index 2 Param 4=25
    Index 3 Param 1=&FloatFromCurve(
     FloatCurve=DataInterface() {propertyValues {Curve {keys [
         {interpMode=RCIM_Cubic, leaveTangent=-0.0025, tangentMode=RCTM_Break, time=0, value=0}
         {arriveTangent=1.8528, interpMode=RCIM_Cubic, tangentMode=RCTM_Break, time=1, value=1}
        ]}}}
     Scale Curve=&RandomRangeFloat(Minimum=0, Maximum=0)
    )
    Index 3 Param 2=&FloatFromCurve(
     FloatCurve=DataInterface() {propertyValues {Curve {keys [
         {interpMode=RCIM_Cubic, leaveTangent=-0.0025, tangentMode=RCTM_Break, time=0, value=0.5}
         {arriveTangent=-0.0316, interpMode=RCIM_Cubic, tangentMode=RCTM_Break, time=1, value=-0.2}
        ]}}}
     Scale Curve=0
    )
    Index 3 Param 3=&FloatFromCurve(FloatCurve=DataInterface() {propertyValues {Curve {keys [
         {interpMode=RCIM_Cubic, leaveTangent=1.1034, tangentMode=RCTM_Break, time=0, value=-1}
         {arriveTangent=2.2399, tangentMode=RCTM_User, time=0.5, value=0.125}
        ]}}})
    Index 3 Param 4=&FloatFromCurve(FloatCurve=DataInterface() {propertyValues {Curve {keys [
         {interpMode=RCIM_Cubic, leaveTangent=1.8089, tangentMode=RCTM_Break, time=0.3, value=-0.5}
         {arriveTangent=2.0757, interpMode=RCIM_Cubic, tangentMode=RCTM_Break, time=0.9, value=1}
        ]}}})
    Module.Index0Param0Write=true
    Module.Index0Param1Write=true
    Module.Index0Param2Write=true
    Module.Index0Param3Write=true
    Module.Index1Param0Write=true
    Module.Index1Param1Write=true
    Module.Index1Param2Write=true
    Module.Index1Param3Write=true
    Module.Index2Param0Write=true
    Module.Index2Param1Write=true
    Module.Index2Param2Write=true
    Module.Index2Param3Write=true
    Module.Index3Param0Write=true
    Module.Index3Param1Write=true
    Module.Index3Param2Write=true
    Module.Index3Param3Write=true
    Write Parameter Index 1=true
    Write Parameter Index 2=true
    Write Parameter Index 3=true
   )
   &ScaleColor(
    Scale Mode=::NewEnumerator1
    Scale RGBA=&Multiply_Vector4(
     A=&MakeVector4_FromLinearColor(LinearColor=$System.GlowColor)
     B=&MakeVector4FromVectorAndFloat(
      "Vector (RGB)"=&Lerp_Vector(
       A=&VectorFromFloat(Value=200)
       B=&VectorFromFloat(Value=&RandomRangeFloat(Minimum=0, Maximum=2, Evaluation Type=::NewEnumerator1))
       Alpha=&FloatFromCurve(FloatCurve=DataInterface() {propertyValues {Curve {keys [
            {interpMode=RCIM_Cubic, leaveTangent=-0.3196, tangentMode=RCTM_Break, time=0, value=0}
            {arriveTangent=6.19, interpMode=RCIM_Cubic, tangentMode=RCTM_Break, time=0.4, value=1}
           ]}}})
      )
      "Float (Alpha)"=&FloatFromCurve(FloatCurve=DataInterface() {propertyValues {Curve {keys [
           {tangentMode=RCTM_User, time=0, value=1}
           {arriveTangent=-4.3986, interpMode=RCIM_Cubic, leaveTangent=-4.3986, tangentMode=RCTM_User, time=0.45, value=1}
           {arriveTangent=0.1205, interpMode=RCIM_Cubic, leaveTangent=0.1205, tangentMode=RCTM_User, time=1, value=0}
          ]}}})
     )
    )
    ScaleRGBA=true
   ) {module=ScaleColor001}
   /SetVariables(Particles.Position=&AddVectorToPosition(
     Position=$Particles.SkeletalMeshLocation001.SampledPosition
     Vector=Linked() {linkedVariable {name=Particles.ShapeLocation.ShapePosition, type=NiagaraPosition}}
    )) {isSetParameters=true}
   /SetVariables(Particles.Color_Glow:LinearColor=&MultiplyLinearColor(
     Linear Color=$System.LightingColor_Additive
     Scale Factor=&MakeVector4FromVectorAndFloat(
      "Vector (RGB)"=&Lerp_Vector(
       A=&VectorFromFloat(Value=&FloatFromCurve(FloatCurve=DataInterface() {propertyValues {Curve {keys [<at 0.2771 1>,<at 0.3786 1>]}}}))
       B=&VectorFromFloat(Value=&RandomRangeFloat(Minimum=0, Maximum=2, Evaluation Type=::NewEnumerator1))
       Alpha=&FloatFromCurve(FloatCurve=DataInterface() {propertyValues {Curve {keys [
            {interpMode=RCIM_Cubic, leaveTangent=-0.3196, tangentMode=RCTM_Break, time=0, value=0}
            {arriveTangent=6.19, interpMode=RCIM_Cubic, tangentMode=RCTM_Break, time=0.4, value=1}
           ]}}})
      )
      "Float (Alpha)"=&FloatFromCurve(FloatCurve=DataInterface() {propertyValues {Curve {keys [
           <at 0 1 RCIM_Cubic>
           {arriveTangent=-1, interpMode=RCIM_Cubic, leaveTangent=-1, time=0.45, value=1}
           {interpMode=RCIM_Cubic, leaveTangent=0.1205, time=1, value=0}
          ]}}})
     )
    )) {isSetParameters=true}
   &KillParticles(Kill Particles=&SetBoolByFloatComparison(A=$Particles.NormalizedAge, B=0.8))
  ]
  rendererClasses [&NiagaraMeshRendererProperties]
  renderers [
   {rendererClass=&NiagaraMeshRendererProperties, rendererIndex=0}
   {rendererClass=&NiagaraMeshRendererProperties, rendererIndex=1}
  ]
 }
 SmallCoins_Lighting {
  emitterUpdateScript [
   &EmitterState(Life Cycle Mode=::NewEnumerator1, Loop Behavior=::NewEnumerator1, Loop Duration Mode=::NewEnumerator1, UseLoopDelay=false)
   &SpawnParticlesFromOtherEmitter(
    Attribute Reader=DataInterface() {propertyValues {EmitterBinding {bindingMode=Other, emitterName=Coins_Above_Player}}}
    Spawn Rate=5
    Use Maximum Age=false
    Use Minimum Age=false
   )
  ]
  particleSpawnScript [
   &InitializeParticle(
    Color=<rgba 2 2 2 1>
    Lifetime=&RandomRangeFloat(Minimum=0.6, Maximum=1)
    Mesh Scale Mode=::NewEnumerator3
    Mesh Uniform Scale=0.1
    UsePositionOffset=false
    Write Material Random=true
   )
   &InitialMeshOrientation(
    Mesh Orientation Mode=::NewEnumerator7
    Rotation=&RandomRangeVector(Minimum=<-0.3 -0.3 -1>, Maximum=<0.3 0.3 1>)
    Use Rotation Vector=true
   )
   &InitialMeshOrientation(Mesh Orientation Mode=::NewEnumerator7, Rotation=<0 0.25 0>, Use Rotation Vector=true) {module=InitialMeshOrientation001}
   &CameraOffset(Camera Offset Amount=10)
   &SampleParticlesFromOtherEmitter(
    Emitter Sampling Mode:ENiagara_EmitterLocSamplingMode=::NewEnumerator0
    Mesh Scale Sampling:ENiagara_DirectReadSamplingMode=::NewEnumerator1
    Particle ID Sampling=::NewEnumerator1
   )
   /SetVariables(Particles.Scale=&Multiply_Vector(
     A=$Output.SampleParticlesFromOtherEmitter.SampledScale
     B=&RandomRangeVector(Minimum=<0.5 0.5 0.75>, Maximum=<0.5 0.5 1>)
    )) {isSetParameters=true}
  ]
  particleUpdateScript [
   &ParticleState()
   &ScaleMeshSize(Scale Factor=&VectorFromFloat(Value=&FloatFromCurve(FloatCurve=DataInterface() {propertyValues {Curve {keys [
          {arriveTangent=1.9115, interpMode=RCIM_Cubic, leaveTangent=1.9115, tangentMode=RCTM_User, time=0, value=0}
          {arriveTangent=0.0224, interpMode=RCIM_Cubic, leaveTangent=0.0224, tangentMode=RCTM_User, time=1, value=1.5}
         ]}}})))
   &DynamicMaterialParameters(
    Index 0 Param 2=&RandomRangeFloat(Minimum=0, Maximum=0)
    Index 0 Param 4=13
    Index 1 Param 1=&FloatFromCurve(FloatCurve=DataInterface() {propertyValues {Curve {keys [<at 0 0.5>,<at 1 1>]}}})
    Index 1 Param 2=2
    Index 1 Param 3=&FloatFromCurve(FloatCurve=DataInterface() {propertyValues {Curve {keys [{interpMode=RCIM_Cubic, tangentMode=RCTM_Break, time=0.5, value=-0.5}, <at 1 0.1>]}}})
    Index 2 Param 1=&Multiply_Float(
     A=&FloatFromCurve(
      FloatCurve=DataInterface() {propertyValues {Curve {keys [{interpMode=RCIM_Cubic, leaveTangent=8.1085, tangentMode=RCTM_Break, time=0.8568, value=1}, {tangentMode=RCTM_User, time=1, value=0}]}}}
      Scale Curve=0.4
     )
     B=&Divide_Float(
      A=&MakeFloatFromVector(Vector=$Particles.Scale, Channel=::NewEnumerator0)
      B=&RandomRangeFloat(Minimum=20, Maximum=25)
     )
    )
    Index 2 Param 2=&FloatFromCurve(
     FloatCurve=DataInterface() {propertyValues {Curve {keys [
         {interpMode=RCIM_Cubic, tangentMode=RCTM_Break, time=0.7, value=0}
         {arriveTangent=-0.0859, interpMode=RCIM_Cubic, tangentMode=RCTM_Break, time=1, value=-1}
        ]}}}
     Scale Curve=50
    )
    Index 2 Param 3=&Waveform(
     Global Input Position=&RandomRangeFloat(Minimum=0, Maximum=15)
     Global Phase=$Particles.Age
     "Waveform [1]"=::NewEnumerator12
     "[1] Phase Scale"=2
     "[1] Frequency"=1
     "[1] Amplitude Min/Max"=<0 1>
    )
    Index 2 Param 4=25
    Index 3 Param 1=&FloatFromCurve(
     FloatCurve=DataInterface() {propertyValues {Curve {keys [
         {interpMode=RCIM_Cubic, leaveTangent=-0.0025, tangentMode=RCTM_Break, time=0, value=0}
         {arriveTangent=1.8528, interpMode=RCIM_Cubic, tangentMode=RCTM_Break, time=1, value=1}
        ]}}}
     Scale Curve=&RandomRangeFloat(Minimum=0, Maximum=0)
    )
    Index 3 Param 2=&FloatFromCurve(
     FloatCurve=DataInterface() {propertyValues {Curve {keys [
         {interpMode=RCIM_Cubic, leaveTangent=-0.0025, tangentMode=RCTM_Break, time=0, value=0.5}
         {arriveTangent=-0.0316, interpMode=RCIM_Cubic, tangentMode=RCTM_Break, time=1, value=-0.2}
        ]}}}
     Scale Curve=0
    )
    Index 3 Param 3=&FloatFromCurve(FloatCurve=DataInterface() {propertyValues {Curve {keys [
         {interpMode=RCIM_Cubic, leaveTangent=1.1034, tangentMode=RCTM_Break, time=0, value=-1}
         {arriveTangent=2.2399, tangentMode=RCTM_User, time=0.5, value=0.125}
        ]}}})
    Index 3 Param 4=&FloatFromCurve(FloatCurve=DataInterface() {propertyValues {Curve {keys [
         {interpMode=RCIM_Cubic, leaveTangent=1.8089, tangentMode=RCTM_Break, time=0.3, value=-0.5}
         {arriveTangent=2.0757, interpMode=RCIM_Cubic, tangentMode=RCTM_Break, time=0.9, value=1}
        ]}}})
    Module.Index0Param0Write=true
    Module.Index0Param1Write=true
    Module.Index0Param2Write=true
    Module.Index0Param3Write=true
    Module.Index1Param0Write=true
    Module.Index1Param1Write=true
    Module.Index1Param2Write=true
    Module.Index1Param3Write=true
    Module.Index2Param0Write=true
    Module.Index2Param1Write=true
    Module.Index2Param2Write=true
    Module.Index2Param3Write=true
    Module.Index3Param0Write=true
    Module.Index3Param1Write=true
    Module.Index3Param2Write=true
    Module.Index3Param3Write=true
    Write Parameter Index 1=true
    Write Parameter Index 2=true
    Write Parameter Index 3=true
   )
   &InheritSourceMovement(ApplyForce=false, ApplyPosition=true, ApplyPositionFalloff=false)
   &SolveForcesAndVelocity(Clamp Velocity=false, Limit Acceleration=false)
   &ScaleColor(
    Scale Mode=::NewEnumerator1
    Scale RGBA=&Multiply_Vector4(
     A=&MakeVector4_FromLinearColor(LinearColor=$System.GlowColor)
     B=&MakeVector4FromVectorAndFloat(
      "Vector (RGB)"=&Lerp_Vector(
       A=&VectorFromFloat(Value=40)
       B=&VectorFromFloat(Value=&RandomRangeFloat(Minimum=0, Maximum=2, Evaluation Type=::NewEnumerator1))
       Alpha=&FloatFromCurve(FloatCurve=DataInterface() {propertyValues {Curve {keys [
            {interpMode=RCIM_Cubic, leaveTangent=-0.3196, tangentMode=RCTM_Break, time=0, value=0}
            {arriveTangent=6.19, interpMode=RCIM_Cubic, tangentMode=RCTM_Break, time=0.4, value=1}
           ]}}})
      )
      "Float (Alpha)"=&FloatFromCurve(FloatCurve=DataInterface() {propertyValues {Curve {keys [
           {tangentMode=RCTM_User, time=0, value=1}
           {arriveTangent=-4.3986, interpMode=RCIM_Cubic, leaveTangent=-4.3986, tangentMode=RCTM_User, time=0.45, value=1}
           {arriveTangent=0.1205, interpMode=RCIM_Cubic, leaveTangent=0.1205, tangentMode=RCTM_User, time=1, value=0}
          ]}}})
     )
    )
    ScaleRGBA=true
   ) {module=ScaleColor001}
   /Niagara/Modules/AttributeReader/UpdateParticlesFromOtherEmitter(Attribute Reader=DataInterface() {propertyValues {EmitterBinding {bindingMode=Other, emitterName=Coins_Above_Player}}})
   /SetVariables(Particles.Color_Glow:LinearColor=&MultiplyLinearColor(
     Linear Color=$System.LightingColor_Additive
     Scale Factor=&MakeVector4FromVectorAndFloat(
      "Vector (RGB)"=&Lerp_Vector(
       A=&VectorFromFloat(Value=&FloatFromCurve(FloatCurve=DataInterface() {propertyValues {Curve {keys [<at 0.2771 1>,<at 0.3786 1>]}}}, Scale Curve=2))
       B=&VectorFromFloat(Value=&RandomRangeFloat(Minimum=0, Maximum=2, Evaluation Type=::NewEnumerator1))
       Alpha=&FloatFromCurve(FloatCurve=DataInterface() {propertyValues {Curve {keys [
            {interpMode=RCIM_Cubic, leaveTangent=-0.3196, tangentMode=RCTM_Break, time=0, value=0}
            {arriveTangent=6.19, interpMode=RCIM_Cubic, tangentMode=RCTM_Break, time=0.4, value=1}
           ]}}})
      )
      "Float (Alpha)"=&FloatFromCurve(FloatCurve=DataInterface() {propertyValues {Curve {keys [
           {tangentMode=RCTM_User, time=0, value=1}
           {arriveTangent=-4.3986, interpMode=RCIM_Cubic, leaveTangent=-4.3986, tangentMode=RCTM_User, time=0.45, value=1}
           {arriveTangent=0.1205, interpMode=RCIM_Cubic, leaveTangent=0.1205, tangentMode=RCTM_User, time=1, value=0}
          ]}}})
     )
    )) {isSetParameters=true}
   &KillParticles(Kill Particles=&SetBoolByFloatComparison(A=$Particles.NormalizedAge, B=0.8))
  ]
  rendererClasses [&NiagaraMeshRendererProperties]
  renderers [
   {rendererClass=&NiagaraMeshRendererProperties, rendererIndex=0}
   {rendererClass=&NiagaraMeshRendererProperties, rendererIndex=1}
  ]
 }
 Unlit_Coin_Under_Player {
  emitterUpdateScript [
   &EmitterState(
    Inactive Response=::NewEnumerator1
    Life Cycle Mode=::NewEnumerator1
    Loop Behavior=::NewEnumerator1
    Loop Delay=2.2
    Loop Duration Mode=::NewEnumerator1
    UseLoopDelay=false
   )
   &SpawnBurst_Instantaneous(Use Spawn Probability=false, UseLoopCountLimit=false)
  ]
  particleSpawnScript [
   &InitializeParticle(
    Color=<rgba 1 1 1 0.4>
    Mesh Scale=&Multiply_VectorByFloat(Vector=<1 0.25 1>, Float=&Multiply_Float(A=1.9, B=$System.BigCoin_Scale))
    Mesh Scale Mode=::NewEnumerator4
    Sprite Size Mode=::NewEnumerator3
    Uniform Sprite Size=210
    UsePositionOffset=false
    Write Material Random=true
   )
   &CameraOffset(Camera Offset Amount=5)
   &InitialMeshOrientation(Mesh Orientation Mode=::NewEnumerator2, Orientation Vector=<0 0 1>, Rotation=<0.5 0 0>, Use Rotation Vector=true)
  ]
  particleUpdateScript [
   &ParticleState(Kill Particles When Lifetime Has Elapsed=false, Let Infinitely Lived Particles Die When Emitter Deactivates=true)
   &SkeletalMeshLocation(
    Bone / Socket Index=4
    Bone Sampling Mode=::NewEnumerator4
    MaskSpawnGroup=false
    Sampled Position Offset=<0 40 0>
    Skeletal Mesh=$System.Skeletal Mesh
   )
   &ScaleMeshSize(Scale Factor=&Multiply_VectorByFloat(
     Vector=<1 1 1>
     Float=&FloatFromCurve(
      FloatCurve=DataInterface() {propertyValues {Curve {
         defaultValue=1
         keys [{interpMode=RCIM_Cubic, tangentMode=RCTM_User, time=0.5, value=0}, {interpMode=RCIM_Cubic, tangentMode=RCTM_User, time=0.75, value=1}]
        }}}
      CurveIndex=$System.Age
     )
    ))
   /SetVariables(Particles.VisibilityTag=$System.VariantIndex) {isSetParameters=true}
   &ScaleColor(
    Scale Mode=::NewEnumerator1
    Scale RGBA=&Multiply_Vector4(
     A=&MakeVector4_FromLinearColor(LinearColor=$System.CoinColors)
     B=&MakeVector4FromVectorAndFloat(
      "Vector (RGB)"=<1 1 1>
      "Float (Alpha)"=&FloatFromCurve(
       FloatCurve=DataInterface() {propertyValues {Curve {
          defaultValue=1
          keys [{interpMode=RCIM_Cubic, tangentMode=RCTM_User, time=0.25, value=0}, {interpMode=RCIM_Cubic, tangentMode=RCTM_User, time=0.6, value=1}]
         }}}
       CurveIndex=$System.Age
       Scale Curve=0.2
      )
     )
    )
    ScaleRGBA=true
   ) {module=ScaleColor001}
  ]
  rendererClasses [&NiagaraMeshRendererProperties]
  renderers [
   {rendererClass=&NiagaraMeshRendererProperties, rendererIndex=0}
   {rendererClass=&NiagaraMeshRendererProperties, rendererIndex=1}
   {rendererClass=&NiagaraMeshRendererProperties, rendererIndex=2}
   {rendererClass=&NiagaraMeshRendererProperties, rendererIndex=3}
   {rendererClass=&NiagaraMeshRendererProperties, rendererIndex=4}
   {rendererClass=&NiagaraMeshRendererProperties, rendererIndex=5}
   {rendererClass=&NiagaraMeshRendererProperties, rendererIndex=6}
  ]
 }
}
summary {
 systemName=NS_Contrail_SatireCane_Ode
 emitters [
  {emitterName=Glow_Under, rendererClasses [&NiagaraMeshRendererProperties]}
  {emitterName=Coin_Under_Player, rendererClasses [&NiagaraMeshRendererProperties]}
  {emitterName=SmallCoins_Lighting, rendererClasses [&NiagaraMeshRendererProperties]}
  {emitterName=Coins_Above_Player, rendererClasses [&NiagaraMeshRendererProperties]}
  {
   emitterName=CoinUnderPlayer_Burst
   rendererClasses [&NiagaraMeshRendererProperties, &NiagaraLightRendererProperties]
  }
  {emitterName=Lighting_Coin_Under_Player, rendererClasses [&NiagaraMeshRendererProperties]}
  {emitterName=Unlit_Coin_Under_Player, rendererClasses [&NiagaraMeshRendererProperties]}
 ]
 userVariables [
  User.bIsFrontEnd {type=NiagaraBool}
  User.bIsFrontEndPreview {type=NiagaraBool}
  User.EmissiveColor {type=LinearColor, description="", defaultValue {struct=LinearColor, value=<rgba 0.1845 0.5029 0.2384 1>}}
  User.ParaGlideLeanAlpha {type=NiagaraFloat, defaultValue=1, description=""}
 ]
}
