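# Level Sequence that plays the SatireCane Ode skydiving contrail Niagara effect; whether it serves preview or gameplay is not shown.
# It possesses a NiagaraActor, with its NiagaraComponent bound as a child.
# A single Niagara system track on the component uses one spawn section to control the system's life cycle over the section's range.
# There are no top-level tracks, and the actor binding itself has none.
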LevelSequence /BRCosmetics/Athena/Effects/Particles/Skydiving/NS_Contrail_SatireCane_Ode/Sequence_Contrail_SatireCane_Ode
bindings [
 NS_Contrail_SatireCane_Ode {id=#1, possessed_class=NiagaraActor, tracks []}
 NiagaraComponent0 {
  id=#2
  parent_id=#1
  possessed_class=NiagaraComponent
  tracks [{type=MovieSceneNiagaraSystemTrack, sections [{type=MovieSceneNiagaraSystemSpawnSection, end_frame=305, start_frame=0}]}]
 }
]
display_rate {denominator=1, numerator=30}
tick_resolution {denominator=1, numerator=24000}
tracks []
